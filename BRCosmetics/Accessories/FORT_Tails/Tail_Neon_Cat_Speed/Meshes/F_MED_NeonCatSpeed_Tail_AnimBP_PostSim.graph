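# Post-simulation Animation Blueprint for the Neon Cat Speed tail cosmetic, adding secondary motion on the shared Fortnite tail skeleton.
# The AnimGraph runs the input pose through a ten-bone Trail controller ending at C_Tail_A_9_Jnt, with blend alpha and relaxation speed driven by variables.
# When character customization loads, it checks for back bling or a cape and, if either is present, sets a dynamic material parameter to hide an element.
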AnimBlueprint /BRCosmetics/Accessories/FORT_Tails/Tail_Neon_Cat_Speed/Meshes/F_MED_NeonCatSpeed_Tail_AnimBP_PostSim
target_skeleton=/Game/Accessories/FORT_Tails/Common/Fortnite_Base_Tail_Skeleton
graphs [
 AnimGraph {nodes [
   n1 = Trail controller - Bone: C_Tail_A_9_Jnt!(
    ComponentPose=Local To Component(LocalPose=Input Pose())
    Alpha=Get TrailAlpha() -> (TrailAlpha:double)
    RelaxationSpeedScale:float=Get RelaxSpeed() -> (RelaxSpeed:double)
   ) -> (Pose) {
    bLimitRotation=true
    chainLength=10
    maxDeltaTime=0.0333
    alphaScaleBiasClamp {bInterpResult=true}
    rotationLimits [
     {limitMax=<180 180 180>}
     {limitMax=<180 180 180>}
     {limitMax=<180 180 180>}
     {limitMax=<180 180 180>}
     {limitMax=<180 180 180>}
     {limitMax=<180 180 180>}
     {limitMax=<180 180 180>}
     {limitMax=<180 180 180>}
     {limitMax=<180 180 180>}
     {limitMax=<180 180 180>}
    ]
    rotationOffsets [<0 0 0>, <0 0 0>, <0 0 0>, <0 0 0>, <0 0 0>, <0 0 0>, <0 0 0>, <0 0 0>, <0 0 0>, <0 0 0>]
    trailBone {boneName=C_Tail_A_9_Jnt}
    trailRelaxationSpeed {editorCurveData {keys [<at 0 10 RCIM_Linear>,<at 0.5107 8.1304 RCIM_Linear>,<at 1 5 RCIM_Linear>]}}
   }
   Output Pose(Result=Component To Local(ComponentPose=@n1)) {layerGroup=None}
  ]}
 EventGraph {nodes [
   Event OnCharacterCustomizationLoaded()
   n1 = CheckForBackbling()
   > Set isWearingBackbling(isWearingBackbling=@n1)
   > Set isWearingCape(isWearingCape=@n1.bIsBackblingLong)
   > Branch(Condition=OR Boolean(A=Get isWearingCape() -> (isWearingCape:bool), B=Get isWearingBackbling()))
   > n2 = Get Dynamic Material Instance(SkeletalMeshComponent=GetPartSkeletalMeshComponent())
   > SetScalarParameterValue(self=@n2, ParameterName=Hide Element 1, Value=1)
  ]}
]
variables [
 isWearingBackbling {type=boolean}
 isWearingCape {type=boolean}
 RelaxSpeed {type=number}
 TrailAlpha {type=number}
]
